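; BP_CalendarDynamicPOIMultiple is a prototype calendar prop that swaps a point of interest between old and new sets of building foundations.
; Its EnableNewFoundations event enables every dynamic foundation in NewFoundations, then disables every one in OldFoundations.
; BeginPlay calls that event only on the authority and only when the DebugLoadNewFoundations flag is set; no other trigger appears in this graph.
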(event EventBeginPlay
  (Networking|SwitchHasAuthority
    (:Authority
      (if (Variables|Default|GetDebugLoadNewFoundations)
        (CallFunction|EnableNewFoundations)))))

(event Custom|EnableNewFoundations
  ;[1 1 1] Enable new foundations
  (for _array_element (Variables|Default|GetNewFoundations)
    (Class|BuildingFoundation|SetDynamicFoundationEnabled _array_element true))
  (for _array_element_2 (Variables|Default|GetOldFoundations)
    (Class|BuildingFoundation|SetDynamicFoundationEnabled _array_element_2 false)))

(fn ConstructionScript)
