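; BP_CalenderDynamicPOI_DisableActors is a calendar-driven dynamic POI Blueprint that hides extra level actors when its default foundation is disabled.
; The DefaultFoundationDisabled event sets the HideActors flag to true.
; That flag's OnRep function loops over the ActorsToHideOnSwap array and hides each actor in game, so the change replicates to clients.
; The actors are only hidden, and no path shows them again.
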(event EventBeginPlay
  (|CallParentFunction) ; This node is disabled and will not be called.\nDrag off pins to build functionality.
)

(event Collision|EventActorBeginOverlap
  (params (OtherActor :type Actor%20Object%20Reference)))

(event EventTick
  (params (DeltaSeconds :type Float%20%28single-precision%29)))

(event EventDefaultFoundationDisabled
  ;[1 1 1] Hide extra actors on default foundation disable
  (Variables|Default|SetHideActors true))

(fn OnRep_HideActors
  (for _array_element (Variables|Default|GetActorsToHideOnSwap)
    (Rendering|SetActorHiddenInGame _array_element true)))

(fn ConstructionScript
  (|CallParentFunction))
